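# Physics asset for the necklace on the medium female Tennis player body, defining simulation bodies for the necklace and its tags.
# Three capsule bodies sit on spine_05, dyn_necklace and dyn_tags.
# Two constraints form a chain from upper spine to necklace to tags, with moderate swing limits and very tight twist limits.
# The tags have a wider second swing range than the necklace.
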PhysicsAsset /Game/Characters/Player/Female/Medium/Bodies/F_MED_Tennis/Meshes/F_MED_Tennis_Necklace_Physics
bodies {
 dyn_necklace [{shapeType=Capsule, length=3.25, radius=4.6159, center=<6.3911 -0.0004 0>, rotation=<rot 90 0 0>}]
 dyn_tags [{shapeType=Capsule, length=2.342, radius=1.4347, center=<0.0088 0.1526 -1.2563>, rotation=<rot -12.6586 -90 1.1284>}]
 spine_05 [{shapeType=Capsule, length=0.1461, radius=1.8684, center=<2.0317 -0.4222 0.1457>, rotation=<rot 0 36.4674 -178.9026>}]
}
constraints [
 {bone1Name=dyn_necklace, bone2Name=spine_05, swing1LimitDegrees=25, swing2LimitDegrees=25, twistLimitDegrees=3}
 {bone1Name=dyn_tags, bone2Name=dyn_necklace, swing1LimitDegrees=25, swing2LimitDegrees=42.1429, twistLimitDegrees=3}
]
