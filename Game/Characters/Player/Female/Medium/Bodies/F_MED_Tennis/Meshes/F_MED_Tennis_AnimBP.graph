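# Post-process style AnimBP for the female medium Tennis body, adding secondary motion to skirt frills, necklace and arm bandana.
# It copies the parent mesh pose, constrains frills to pelvis and thighs, runs two RigidBody sims and Anim Dynamics, then caches the result.
# Crouch state selects between corrective flap/frill bone offsets, with crouch and speed driving physics and offset alphas.
# When DBNO, a final constraint centers the back frill between its side bones.
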AnimBlueprint /Game/Characters/Player/Female/Medium/Bodies/F_MED_Tennis/Meshes/F_MED_Tennis_AnimBP
target_skeleton=&/F_MED_Tennis_Skeleton
graphs [
 AnimGraph {nodes [
   n1 = "Transform (Modify) Bone - Bone: Flap_Front_Driven_Bone"(
    ComponentPose=Local To Component(LocalPose=Use cached pose 'PreCrouch'())
    Alpha=Get crouch_speed_alpha() -> (crouch_speed_alpha:double)
    Translation="0,0,0"
    Rotation="32.680,26.356,-52.200"
    Scale="1,1,1"
   ) {
    rotationMode=BMM_Additive
    rotationSpace=BCS_BoneSpace
    translationMode=BMM_Additive
    translationSpace=BCS_BoneSpace
    alphaBoolBlend {blendOption=Linear}
    alphaScaleBiasClamp {bInterpResult=true}
    boneToModify {boneName=Flap_Front_Driven_Bone}
    rotation=<rot 32.68 26.356 -52.2>
   }
   n2 = Component To Local(ComponentPose=@n1)
   n3 = "Transform (Modify) Bone - Bone: driver_frill_frt_l"(
    ComponentPose=Local To Component(LocalPose=Use cached pose 'PreCrouch'())
    Alpha=1
    Translation="0,0,0"
    Rotation="0.000,-15,0.000"
    Scale="1,1,1"
   ) {
    rotationMode=BMM_Additive
    rotationSpace=BCS_BoneSpace
    alphaBoolBlend {blendOption=Linear}
    boneToModify {boneName=driver_frill_frt_l}
    rotation {yaw=-15}
   }
   n4 = "Transform (Modify) Bone - Bone: driver_frill_frt_r"(ComponentPose=@n3, Alpha=1, Translation="0,0,0", Rotation="0.000,-24.999,0.000", Scale="1,1,1") {
    rotationMode=BMM_Additive
    rotationSpace=BCS_BoneSpace
    alphaBoolBlend {blendOption=Linear}
    boneToModify {boneName=driver_frill_frt_r}
    rotation {yaw=-24.9993}
   }
   n5 = "Transform (Modify) Bone - Bone: Flap_Front_Parent_Space"(ComponentPose=@n4, Alpha=1, Translation="0,0,0", Rotation="0.000,0,-55", Scale="1,1,1") {
    rotationMode=BMM_Additive
    rotationSpace=BCS_BoneSpace
    alphaBoolBlend {blendOption=Linear}
    boneToModify {boneName=Flap_Front_Parent_Space}
    rotation {roll=-55}
   }
   n6 = Component To Local(ComponentPose=@n5)
   n7 = Get is_crouching() -> (is_crouching:bool)
   n8 = "Constraint - dyn_frill_bck (PP)"!(
    ComponentPose=Local To Component(LocalPose=Blend Poses by bool(BlendPose_0=@n2, BlendPose_1=@n6, bActiveValue=@n7) {blendType=Linear})
    Alpha="To Float (Boolean)"(InBool=Get IsDBNO() -> (IsDBNO:bool))
    ConstraintWeights_0=0.5
    ConstraintWeights_1:float=0.5
   ) -> (Pose) {
    alphaBoolBlend {blendOption=Linear}
    boneToModify {boneName=dyn_frill_bck}
    constraintSetup [{targetBone {boneName=dyn_frill_bck_l}}, {targetBone {boneName=dyn_frill_bck_r}}]
    constraintWeights [1,1]
   }
   Output Pose(Result=Component To Local(ComponentPose=@n8)) {layerGroup=None}
   n9 = "Constraint - dyn_frill_frt (PPPPP)"!(
    ComponentPose=Local To Component(LocalPose=Copy Pose From Mesh() {bCopyCurves=true, bUseAttachedParent=true})
    Alpha=1
    ConstraintWeights_0=0.1453
    ConstraintWeights_1:float=0.1503
    ConstraintWeights_2:float=0.3548
    ConstraintWeights_3:float=0.2236
    ConstraintWeights_4:float=0.126
   ) -> (Pose) {
    alphaBoolBlend {blendOption=Linear}
    boneToModify {boneName=dyn_frill_frt}
    constraintSetup [
     {targetBone {boneName=thigh_r}}
     {targetBone {boneName=pelvis}}
     {targetBone {boneName=thigh_twist_01_l}}
     {targetBone {boneName=thigh_twist_01_r}}
     {targetBone {boneName=thigh_l}}
    ]
    constraintWeights [1,1,1,1,1]
   }
   n10 = "Constraint - dyn_frill_bck (PPPPP)"!(
    ComponentPose=@n9
    Alpha=1
    ConstraintWeights_0=0.0484
    ConstraintWeights_1:float=0.4698
    ConstraintWeights_2:float=0.2345
    ConstraintWeights_3:float=0.2005
    ConstraintWeights_4:float=0.0468
   ) -> (Pose) {
    alphaBoolBlend {blendOption=Linear}
    boneToModify {boneName=dyn_frill_bck}
    constraintSetup [
     {targetBone {boneName=thigh_r}}
     {targetBone {boneName=pelvis}}
     {targetBone {boneName=thigh_twist_01_l}}
     {targetBone {boneName=thigh_twist_01_r}}
     {targetBone {boneName=thigh_l}}
    ]
    constraintWeights [1,1,1,1,1]
   }
   n11 = RigidBody(ComponentPose=@n10, Alpha=Get crouch_rigid_alpha() -> (crouch_rigid_alpha:double)) {
    bOverrideWorldGravity=true
    overridePhysicsAsset=&/F_MED_Tennis_Physics
    simulationSpace=BaseBoneSpace
    alphaBoolBlend {blendOption=Linear}
    baseBoneRef {boneName=root}
    componentAppliedLinearAccClamp=<1000 1000 1000>
    componentLinearAccScale=<1 1 1>
    componentLinearVelScale=<1 1 1>
    overrideWorldGravity {z=-980}
   }
   n12 = RigidBody(ComponentPose=@n11) {
    bOverrideWorldGravity=true
    overridePhysicsAsset=&/F_MED_Tennis_Necklace_Physics
    simulationSpace=BaseBoneSpace
    alphaBoolBlend {blendOption=Linear}
    baseBoneRef {boneName=spine_05}
    componentAppliedLinearAccClamp=<500 500 500>
    componentLinearAccScale=<1 1 1>
    componentLinearVelScale=<1 1 1>
    overrideWorldGravity {z=-980}
   }
   n13 = Anim Dynamics - Bone: dyn_arm_r_bandana(ComponentPose=@n12, Alpha=0.75, ExternalForce="0,0,0") {
    alpha=0.9
    angularDampingOverride=0.98
    angularSpringConstant=1500
    bAngularSpring=true
    bGravityOverrideInSimSpace=true
    bOverrideAngularDamping=true
    gravityScale=0.33
    numSolverIterationsPostUpdate=4
    numSolverIterationsPreUpdate=16
    simulationSpace=RootRelative
    alphaBoolBlend {blendOption=Linear}
    boundBone {boneName=dyn_arm_r_bandana}
    physicsBodyDefinitions [{
      sphereCollisionRadius=0
      boundBone {boneName=dyn_arm_r_bandana}
      boxExtents=<15 5 5>
      constraintSetup {angularTargetAxis=AxisZ, twistAxis=AxisY, angularLimitsMax=<15 0 15>, angularLimitsMin=<-15 0 -15>, angularTarget=<0 0 1>}
      localJointOffset=<0 0 0>
     }]
   }
   Save cached pose 'PreCrouch'(Pose=Component To Local(ComponentPose=@n13)) {cacheName=PreCrouch}
  ]}
 EventGraph {nodes [
   Event BlueprintUpdateAnimation()
   > n1 = Sequence()
   > n2 = Branch(Condition=Get bIsCrouching())
   n3 = Get PawnSpeed2D()
   > n4 = Branch(Condition="float <= float"!(A:double=MapRangeClamped(Value=@n3, InRangeA=0, InRangeB=150), B:double=0) -> (ReturnValue:bool) {is=title K2Node_CallFunction})
   > Set crouch_speed_alpha(crouch_speed_alpha:double=1) -> (Output_Get:double)
   > Set crouch_moving(crouch_moving:bool=false) -> (Output_Get:bool)
   > Set crouch_rigid_alpha(crouch_rigid_alpha:double=0.2) -> (Output_Get:double)
   > Set move_speed_alpha(move_speed_alpha:double=0) -> (Output_Get:double)
   Set crouch_speed_alpha(execute=@n4.else, crouch_speed_alpha=0)
   > Set crouch_moving(crouch_moving=true)
   > Set crouch_rigid_alpha(crouch_rigid_alpha=0.55)
   > Set move_speed_alpha(move_speed_alpha=1)
   Set crouch_rigid_alpha(execute=@n2.else, crouch_rigid_alpha=0.8)
   n5 = TryGetFortPawn()
   Is Valid(exec=@n1.then_1, InputObject=@n5)
   n6 = Get bIsDBNO!(self:Fort Pawn Object Reference=@n5) -> (bIsDBNO)
   > Set IsDBNO(IsDBNO:bool=@n6) -> (Output_Get:bool)
   > Set NotDBNO(NotDBNO:bool=NOT Boolean(A=@n6)) -> (Output_Get:bool)
   Event OnBeginCrouching()
   > Set is_crouching(is_crouching:bool=true) -> (Output_Get:bool)
   Event OnEndCrouching()
   > Set is_crouching(is_crouching=false)
   TryGetPawnOwner()
  ]}
]
variables [
 crouch_moving {type=boolean}
 crouch_rigid_alpha {type=number, default=1}
 crouch_speed_alpha {type=number}
 is_crouching {type=boolean}
 IsDBNO {type=boolean}
 move_speed_alpha {type=number}
 NotDBNO {type=boolean}
]
